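# Animation Blueprint for the F_MED_PiperShelf_Empress player body cosmetic, layering a Control Rig over the pose copied from its attached parent mesh.
# The rig's output is the final pose, and it receives LargeGunEquipped as an input.
# OnWeaponChanged resets that value to 0, then sets it to 1 only for the AssaultRifle weapon core animation.
# The update event is empty, and ar_firingPosition_GC and Gun_out are unused in these graphs.
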AnimBlueprint /BRCosmetics/Characters/Player/Female/Medium/Bodies/F_MED_PiperShelf_Empress/Meshes/F_MED_PiperShelf_Empress_AnimBP
target_skeleton=&/F_MED_PiperShelf_Empress_Skeleton
graphs [
 AnimGraph {nodes [
   n1 = Control Rig!(Source=Copy Pose From Mesh() {bUseAttachedParent=true}, Alpha, LargeGunEquipped:double=Get LargeGunEquipped() -> (LargeGunEquipped:double)) -> (Pose) {
    alphaBoolBlend {blendOption=Linear}
    controlRigAssetReference {blueprintRigClass=&/F_MED_PiperShelf_Empress_CtrlRig.F_MED_PiperShelf_Empress_CtrlRig_C}
   }
   Output Pose(Result=@n1) {layerGroup=None}
  ]}
 EventGraph {nodes [
   Event OnWeaponChanged()
   > Set LargeGunEquipped(LargeGunEquipped:double=0) -> (Output_Get:double)
   > n1 = Switch on EFortWeaponCoreAnimation!(execute, Selection:EFortWeaponCoreAnimation Enum=Get WeaponCoreAnim!() -> (WeaponCoreAnim)) -> (
    Melee:Exec
    Pistol:Exec
    Shotgun:Exec
    PaperBlueprint:Exec
    Rifle:Exec
    MeleeOneHand:Exec
    MachinePistol:Exec
    RocketLauncher:Exec
    GrenadeLauncher:Exec
    GoingCommando:Exec
    AssaultRifle:Exec
    TacticalShotgun:Exec
    SniperRifle:Exec
    TrapPlacement:Exec
    ShoulderLauncher:Exec
    AbilityDecoTool:Exec
    Crossbow:Exec
    C4:Exec
    RemoteControl:Exec
    DualWield:Exec
    AR_BullPup:Exec
    AR_ForwardGrip:Exec
    MedPackPaddles:Exec
    SMG_P90:Exec
    AR_DrumGun:Exec
    Consumable_Small:Exec
    Consumable_Large:Exec
    Balloon:Exec
    MountedTurret:Exec
    CreativeTool:Exec
    ExplosiveBow:Exec
    AshtonIndigo:Exec
    AshtonChicago:Exec
    MeleeDualWield:Exec
    Unarmed:Exec
   )
   Set LargeGunEquipped(execute=@n1.AssaultRifle, LargeGunEquipped=1)
   Event BlueprintUpdateAnimation()
  ]}
]
variables [
 ar_firingPosition_GC {type=integer}
 Gun_out {type=boolean}
 LargeGunEquipped {type=number}
]
