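# Control Rig for the F_MED_PiperShelf_Empress face accessory that applies corrective poses to hair bones as the head, arms and aim change.
# Spherical pose readers on head and upperarms produce alphas, with head alphas faded while emoting.
# These, plus head pitch relative to spine_05 and targeting, weapon and aim-pitch inputs, weight additive ModifyTransforms.
# PickaxeSwing is never set here and defaults to zero, muting back-hair look-left/right correctives.
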ControlRigBlueprint /BRCosmetics/Characters/Player/Female/Medium/Bodies/F_MED_PiperShelf_Empress/Meshes/Parts/F_MED_PiperShelf_Empress_FaceAcc_CtrlRig
graphs [
 Multiply_1_ContainedGraph {nodes [
   n1 = RigVMFunctionEntryNode()
   Multiply = &RigVMFunction_MathDoubleMul(A=@n1.A, B=@n1.B)
   RigVMFunctionReturnNode(Result=&RigVMFunction_MathDoubleMul(B=@n1.C, A=@Multiply.Result).Result)
  ]}
 Multiply_1_ContainedGraph {nodes [
   n1 = RigVMFunctionEntryNode()
   Multiply = &RigVMFunction_MathDoubleMul(A=@n1.A, B=@n1.B)
   RigVMFunctionReturnNode(Result=&RigVMFunction_MathDoubleMul(B=@n1.C, A=@Multiply.Result).Result)
  ]}
 New Function_ContainedGraph {nodes [
   n1 = RigVMFunctionEntryNode()
   RigVMFunction_ControlFlowBranch = &RigVMFunction_ControlFlowBranch(Condition=@n1.Debug, ExecuteContext=@n1.ExecuteContext)
   RigVMFunctionReturnNode(ExecuteContext=@RigVMFunction_ControlFlowBranch.Completed)
   RigUnit_GetTransform = &RigUnit_GetTransform(Item=(Type=Control,Name="head_ctrl"), bInitial=False)
   RigUnit_SetTransform = &RigUnit_SetTransform(
    Item=(Type=Bone,Name="head")
    Weight=1
    bInitial=False
    Value=@RigUnit_GetTransform.Transform
    ExecutePin=@RigVMFunction_ControlFlowBranch.True
   )
   > &RigUnit_SetTransform(
    Item=(Type=Bone,Name="clavicle_l")
    Weight=1
    bInitial=False
    Value=&RigUnit_GetTransform(Item=(Type=Control,Name="clavicle_l_ctrl"), bInitial=False).Transform
   )
   > &RigUnit_SetTransform(
    Item=(Type=Bone,Name="upperarm_l")
    Weight=1
    bInitial=False
    Value=&RigUnit_GetTransform(Item=(Type=Control,Name="upperarm_l_ctrl"), bInitial=False).Transform
   )
   > &RigUnit_SetTransform(
    Item=(Type=Bone,Name="clavicle_r")
    Weight=1
    bInitial=False
    Value=&RigUnit_GetTransform(Item=(Type=Control,Name="clavicle_r_ctrl"), bInitial=False).Transform
   )
   > &RigUnit_SetTransform(
    Item=(Type=Bone,Name="upperarm_r")
    Weight=1
    bInitial=False
    Value=&RigUnit_GetTransform(Item=(Type=Control,Name="upperarm_r_ctrl"), bInitial=False).Transform
   )
  ]}
 New Function_ContainedGraph {nodes [
   RigVMFunctionEntryNode()
   > RigVMFunction_Sequence = RigVMAggregateNode()
   RigVMFunctionReturnNode(ExecuteContext=@RigVMFunction_Sequence.C)
   n1 = &RigUnit_SphericalPoseReader(
    ActiveRegionScaleFactors=(PositiveWidth=2,NegativeWidth=2,PositiveHeight=1,NegativeHeight=1)
    ActiveRegionSize=0.05
    Debug=(bDrawDebug=false,bDraw2D=False,bDrawLocalAxes=False,DebugScale=25,DebugSegments=20,DebugThickness=0.25)
    DriverAxis=(X=1,Y=0,Z=0)
    DriverItem=(Type=Bone,Name="upperarm_l")
    FalloffRegionScaleFactors=(PositiveWidth=1,NegativeWidth=1,PositiveHeight=0.75,NegativeHeight=0.75)
    FalloffSize=0.2
    FlipHeightScaling=False
    FlipWidthScaling=False
    OptionalParentItem=(Type=Bone,Name="None")
    RotationOffset=(X=-15,Y=-50,Z=0)
    ExecutePin=@RigVMFunction_Sequence.B
   )
   > RigVMVariableNode(Value=@n1.OutputParam) {variable=armUp_l_alpha}
   > n2 = &RigUnit_SphericalPoseReader(
    ActiveRegionScaleFactors=(PositiveWidth=2,NegativeWidth=2,PositiveHeight=1,NegativeHeight=1)
    ActiveRegionSize=0.05
    Debug=(bDrawDebug=false,bDraw2D=False,bDrawLocalAxes=False,DebugScale=25,DebugSegments=20,DebugThickness=0.25)
    DriverAxis=(X=-1,Y=0,Z=0)
    DriverItem=(Type=Bone,Name="upperarm_r")
    FalloffRegionScaleFactors=(PositiveWidth=1,NegativeWidth=1,PositiveHeight=0.75,NegativeHeight=0.75)
    FalloffSize=0.2
    FlipHeightScaling=False
    FlipWidthScaling=false
    OptionalParentItem=(Type=Bone,Name="None")
    RotationOffset=(X=15,Y=130,Z=0)
   )
   > RigVMVariableNode(Value=@n2.OutputParam) {variable=armUp_r_alpha}
   n3 = &RigUnit_SphericalPoseReader(
    ActiveRegionScaleFactors=(PositiveWidth=1,NegativeWidth=0.5,PositiveHeight=1.5,NegativeHeight=1.5)
    ActiveRegionSize=0.1
    Debug=(bDrawDebug=false,DebugScale=25,DebugSegments=20,DebugThickness=0.25)
    DriverAxis=(X=0,Y=1,Z=0)
    DriverItem=(Type=Bone,Name="head")
    FalloffRegionScaleFactors=(PositiveWidth=1,NegativeWidth=1.8,PositiveHeight=1.5,NegativeHeight=1.5)
    FalloffSize=0.2
    FlipWidthScaling=false
    OptionalParentItem=(Type=Bone)
    RotationOffset=(X=90,Y=0,Z=-60)
    ExecutePin=@RigVMFunction_Sequence.A
   )
   > RigVMVariableNode(Value=@n3.OutputParam) {variable=head_lookup_alpha}
   > n4 = &RigUnit_SphericalPoseReader(
    ActiveRegionScaleFactors=(PositiveWidth=2,NegativeWidth=0.5,PositiveHeight=2,NegativeHeight=2)
    ActiveRegionSize=0.075
    Debug=(bDrawDebug=false,DebugScale=25,DebugSegments=20,DebugThickness=0.25)
    DriverAxis=(X=0,Y=1,Z=0)
    DriverItem=(Type=Bone,Name="head")
    FalloffRegionScaleFactors=(PositiveWidth=1,NegativeWidth=1.1,PositiveHeight=1,NegativeHeight=1)
    FalloffSize=0.2
    FlipHeightScaling=false
    FlipWidthScaling=true
    OptionalParentItem=(Type=Bone)
    RotationOffset=(X=90,Y=0,Z=42.5)
   )
   > RigVMVariableNode(Value=@n4.OutputParam) {variable=head_lookdown_alpha}
   > n5 = &RigUnit_SphericalPoseReader(
    ActiveRegionScaleFactors=(PositiveWidth=0.75,NegativeWidth=0.75,PositiveHeight=1,NegativeHeight=1)
    ActiveRegionSize=0.1
    Debug=(bDrawDebug=false,DebugScale=25,DebugSegments=20,DebugThickness=0.25)
    DriverAxis=(X=0,Y=1,Z=0)
    DriverItem=(Type=Bone,Name="head")
    FalloffRegionScaleFactors=(PositiveWidth=1,NegativeWidth=1,PositiveHeight=1,NegativeHeight=2.2)
    FalloffSize=0.2
    OptionalParentItem=(Type=Bone)
    RotationOffset=(X=150,Y=0,Z=0)
   )
   > RigVMVariableNode(Value=@n5.OutputParam) {variable=head_lookleft_alpha}
   > n6 = &RigUnit_SphericalPoseReader(
    ActiveRegionScaleFactors=(PositiveWidth=0.75,NegativeWidth=0.75,PositiveHeight=1,NegativeHeight=1)
    ActiveRegionSize=0.1
    Debug=(bDrawDebug=false,DebugScale=25,DebugSegments=20,DebugThickness=0.25)
    DriverAxis=(X=0,Y=1,Z=0)
    DriverItem=(Type=Bone,Name="head")
    FalloffRegionScaleFactors=(PositiveWidth=1,NegativeWidth=1,PositiveHeight=1,NegativeHeight=2.2)
    FalloffSize=0.2
    FlipHeightScaling=true
    FlipWidthScaling=false
    OptionalParentItem=(Type=Bone)
    RotationOffset=(X=30,Y=0,Z=0)
   )
   > RigVMVariableNode(Value=@n6.OutputParam) {variable=head_lookright_alpha}
  ]}
 New Function_ContainedGraph {nodes [
   RigVMFunctionEntryNode()
   > RigVMFunction_Sequence = RigVMAggregateNode()
   RigVMFunctionReturnNode(ExecuteContext=@RigVMFunction_Sequence.C)
   n1 = &RigVMFunction_MathQuaternionToEuler(
    RotationOrder=ZYX
    Value=&RigUnit_GetRelativeTransformForItem(Child=(Type=Bone,Name="head"), Parent=(Type=Bone,Name="spine_05"), bChildInitial=False, bParentInitial=False).RelativeTransform.Rotation
   )
   RigVMVariableNode(
    Value=&RigVMFunction_MathDoubleRemap(
     SourceMaximum=-10
     SourceMinimum=-40
     TargetMaximum=1
     TargetMinimum=0
     Value=&RigVMFunction_MathDoubleAdd(B=-4.1228, A=@n1.Result.Z).Result
    ).Result
    ExecuteContext=@RigVMFunction_Sequence.B
   ) {variable=head_up_relative}
   n2 = &RigVMFunction_MathDoubleRemap(SourceMaximum=1, SourceMinimum=0, TargetMaximum=0, TargetMinimum=1, bClamp=False, Value=RigVMVariableNode() {variable=pub isEmoting}.Value)
   RigVMVariableNode(
    Value=&RigVMFunction_MathDoubleMul(A=RigVMVariableNode() {variable=head_lookup_alpha}.Value, B=@n2.Result).Result
    ExecuteContext=@RigVMFunction_Sequence.A
   ) {variable=head_lookup_alpha}
   > RigVMVariableNode(Value=&RigVMFunction_MathDoubleMul(A=RigVMVariableNode() {variable=head_lookdown_alpha}.Value, B=@n2.Result).Result) {variable=head_lookdown_alpha}
   > RigVMVariableNode(Value=&RigVMFunction_MathDoubleMul(A=RigVMVariableNode() {variable=head_lookright_alpha}.Value, B=@n2.Result).Result) {variable=head_lookright_alpha}
   > RigVMVariableNode(Value=&RigVMFunction_MathDoubleMul(A=RigVMVariableNode() {variable=head_lookleft_alpha}.Value, B=@n2.Result).Result) {variable=head_lookleft_alpha}
  ]}
 New Function_ContainedGraph {nodes [
   RigVMFunctionEntryNode()
   > RigVMFunction_Sequence = RigVMAggregateNode()
   RigVMFunctionReturnNode(ExecuteContext=@RigVMFunction_Sequence.D)
   &RigUnit_ModifyTransforms(
    ItemToModify="((Item=(Type=Bone,Name=\"base_hairB_C\"),Transform=(Rotation=(X=0,Y=0,Z=-0.5736,W=0.8192),Translation=(X=0,Y=0,Z=0),Scale3D=(X=1,Y=1,Z=1))),(Item=(Type=Bone,Name=\"base_hairB_L\"),Transform=(Rotation=(X=0,Y=0,Z=-0.5736,W=0.8192),Translation=(X=0,Y=0,Z=0),Scale ...(truncated)"
    WeightMinimum=0
    Weight=RigVMVariableNode() {variable=head_lookup_alpha}.Value
    ExecutePin=@RigVMFunction_Sequence.A
   )
   > &RigUnit_ModifyTransforms(
    ItemToModify="((Item=(Type=Bone,Name=\"base_hairB_C\"),Transform=(Rotation=(X=0,Y=0,Z=0.1736,W=0.9848),Translation=(X=0,Y=0,Z=0),Scale3D=(X=1,Y=1,Z=1))),(Item=(Type=Bone,Name=\"base_hairB_L\"),Transform=(Rotation=(X=0,Y=0,Z=0.1736,W=0.9848),Translation=(X=0,Y=0,Z=0),Scale3D ...(truncated)"
    WeightMinimum=0
    Weight=RigVMVariableNode() {variable=head_lookdown_alpha}.Value
   )
   > &RigUnit_ModifyTransforms(
    ItemToModify="((Item=(Type=Bone,Name=\"base_hairB_C\"),Transform=(Rotation=(X=-0.3925,Y=-0.1279,Z=-0.0003,W=0.9108),Translation=(X=0,Y=0,Z=0),Scale3D=(X=1,Y=1,Z=1))),(Item=(Type=Bone,Name=\"base_hairB_L\"),Transform=(Rotation=(X=-0.3304,Y=-0.1321,Z=0.1114,W=0.9279),Translat ...(truncated)"
    WeightMinimum=0
    Weight=&RigVMFunction_MathDoubleMul(B=RigVMVariableNode() {variable=PickaxeSwing}.Value, A=RigVMVariableNode() {variable=head_lookleft_alpha}.Value).Result
   )
   > &RigUnit_ModifyTransforms(
    ItemToModify="((Item=(Type=Bone,Name=\"base_hairB_C\"),Transform=(Rotation=(X=0.6951,Y=0.2833,Z=-0.1649,W=0.6398),Translation=(X=0,Y=0,Z=0),Scale3D=(X=1,Y=1,Z=1))),(Item=(Type=Bone,Name=\"base_hairB_L\"),Transform=(Rotation=(X=0.3486,Y=0.392,Z=-0.3248,W=0.787),Translation=( ...(truncated)"
    WeightMinimum=0
    Weight=&RigVMFunction_MathDoubleMul(B=RigVMVariableNode() {variable=PickaxeSwing}.Value, A=RigVMVariableNode() {variable=head_lookright_alpha}.Value).Result
   )
   &RigUnit_ModifyTransforms(
    ItemToModify="((Item=(Type=Bone,Name=\"base_hairB_R\"),Transform=(Rotation=(X=-0.2588,Y=0.1311,Z=-0.3518,W=0.89),Translation=(X=0,Y=0,Z=0),Scale3D=(X=1,Y=1,Z=1))),(Item=(Type=Bone,Name=\"dyn_hairB_R_02\"),Transform=(Rotation=(X=0.1786,Y=-0.0325,Z=0.3248,W=0.9282),Translatio ...(truncated)"
    WeightMinimum=0
    Weight=RigVMVariableNode() {variable=armUp_r_alpha}.Value
    ExecutePin=@RigVMFunction_Sequence.B
   )
   > &RigUnit_ModifyTransforms(
    ItemToModify="((Item=(Type=Bone,Name=\"base_hairB_L\"),Transform=(Rotation=(X=0.2049,Y=-0.0126,Z=-0.2352,W=0.95),Translation=(X=0,Y=0,Z=0),Scale3D=(X=1,Y=1,Z=1))),(Item=(Type=Bone,Name=\"dyn_hairB_L_02\"),Transform=(Rotation=(X=0.0066,Y=0.0032,Z=0.1436,W=0.9896),Translation ...(truncated)"
    WeightMinimum=0
    Weight=RigVMVariableNode() {variable=armUp_l_alpha}.Value
   )
   &RigUnit_ModifyTransforms(
    ItemToModify="((Item=(Type=Bone,Name=\"base_hairF_L_a\"),Transform=(Rotation=(X=0,Y=-0.4226,Z=0,W=0.9063),Translation=(X=0,Y=0,Z=0),Scale3D=(X=1,Y=1,Z=1))),(Item=(Type=Bone,Name=\"base_hairF_R_a\"),Transform=(Rotation=(X=0,Y=-0.4226,Z=0,W=0.9063),Translation=(X=0,Y=0,Z=0),S ...(truncated)"
    WeightMinimum=0
    Weight=RigVMVariableNode() {variable=head_lookup_alpha}.Value
    ExecutePin=@RigVMFunction_Sequence.C
   )
   > &RigUnit_ModifyTransforms(
    ItemToModify="((Item=(Type=Bone,Name=\"base_hairF_L_a\"),Transform=(Rotation=(X=0,Y=0.1736,Z=0,W=0.9848),Translation=(X=0,Y=0,Z=0),Scale3D=(X=1,Y=1,Z=1))),(Item=(Type=Bone,Name=\"base_hairF_R_a\"),Transform=(Rotation=(X=0,Y=0.1736,Z=0,W=0.9848),Translation=(X=0,Y=0,Z=0),Sca ...(truncated)"
    WeightMinimum=0
    Weight=RigVMVariableNode() {variable=head_lookdown_alpha}.Value
   )
   > &RigUnit_ModifyTransforms(
    ItemToModify="((Item=(Type=Bone,Name=\"base_hairF_L_a\"),Transform=(Rotation=(X=0.3261,Y=0.2489,Z=-0.0727,W=0.9091),Translation=(X=0,Y=0,Z=2),Scale3D=(X=1,Y=1,Z=1))),(Item=(Type=Bone,Name=\"base_hairF_R_a\"),Transform=(Rotation=(X=0,Y=0.1736,Z=0,W=0.9848),Translation=(X=0,Y ...(truncated)"
    WeightMinimum=0
    Weight=&RigVMFunction_MathDoubleMul(A=RigVMVariableNode() {variable=head_up_relative}.Value, B=RigVMVariableNode() {variable=head_lookleft_alpha}.Value).Result
   )
   > &RigUnit_ModifyTransforms(
    ItemToModify="((Item=(Type=Bone,Name=\"base_hairF_L_a\"),Transform=(Rotation=(X=0,Y=0.1736,Z=0,W=0.9848),Translation=(X=0,Y=0,Z=1),Scale3D=(X=1,Y=1,Z=1))),(Item=(Type=Bone,Name=\"base_hairF_R_a\"),Transform=(Rotation=(X=0.3346,Y=0.2374,Z=-0.1043,W=0.906),Translation=(X=0,Y= ...(truncated)"
    WeightMinimum=0
    Weight=&RigVMFunction_MathDoubleMul(A=RigVMVariableNode() {variable=head_up_relative}.Value, B=RigVMVariableNode() {variable=head_lookright_alpha}.Value).Result
   )
   Clamp = &RigVMFunction_MathDoubleClamp(Value=&RigVMFunction_MathDoubleMul(B=2.2, A=RigVMVariableNode() {variable=head_lookup_alpha}.Value).Result)
   Multiply = RigVMAggregateNode(C=@Clamp.Result, B=RigVMVariableNode() {variable=pub isUsingLargeWeapons}.Value, A=RigVMVariableNode() {variable=pub isTargeting}.Value)
   > &RigUnit_ModifyTransforms(
    ItemToModify="((Item=(Type=Bone,Name=\"base_hairB_L\"),Transform=(Rotation=(X=-0.2496,Y=-0.0226,Z=-0.193,W=0.9486),Translation=(X=1.1,Y=8.5,Z=0.75),Scale3D=(X=1,Y=1,Z=1))),(Item=(Type=Bone,Name=\"base_hairB_C\"),Transform=(Rotation=(X=0,Y=0,Z=-0.1554,W=0.9879),Translation=( ...(truncated)"
    WeightMinimum=0
    Weight=@Multiply.Result
   )
   Multiply_4 = RigVMAggregateNode(
    C=&RigVMFunction_MathDoubleRemap(SourceMaximum=0, SourceMinimum=-90, TargetMaximum=0, TargetMinimum=1, Value=RigVMVariableNode() {variable=pub aimPitch}.Value).Result
    A=RigVMVariableNode() {variable=pub isTargeting}.Value
    B=RigVMVariableNode() {variable=pub isUsingMingun}.Value
   )
   n1 = &RigVMFunction_AnimEvalRichCurve(
    Curve=(EditorCurveData=(Keys=((InterpMode=RCIM_Cubic),(InterpMode=RCIM_Cubic,TangentMode=RCTM_User,Time=0.4302,Value=0.2017,ArriveTangent=0.8571,LeaveTangent=0.8571),(Time=1,Value=1))))
    SourceMinimum=0
    TargetMinimum=0
    Value=@Multiply_4.Result
   )
   > &RigUnit_ModifyTransforms(
    ItemToModify="((Item=(Type=Bone,Name=\"base_hairB_C\"),Transform=(Rotation=(X=0.0027,Y=-0.2335,Z=-0.0112,W=0.9723),Translation=(X=0.8483,Y=-0.4526,Z=0.5098),Scale3D=(X=1,Y=1,Z=1))),(Item=(Type=Bone,Name=\"base_hairB_R\"),Transform=(Rotation=(X=-0.244,Y=-0.2341,Z=-0.0471,W=0 ...(truncated)"
    WeightMinimum=0
    Weight=@n1.Result
   )
   n2 = &RigUnit_SphericalPoseReader(
    ActiveRegionScaleFactors=(PositiveWidth=1,NegativeWidth=1,PositiveHeight=1,NegativeHeight=1)
    ActiveRegionSize=0.06
    Debug=(bDrawDebug=true,DebugScale=25,DebugSegments=20,DebugThickness=0.25)
    DriverAxis=(X=-1,Y=0,Z=0)
    DriverItem=(Type=Bone,Name="clavicle_r")
    ExecutePin=()
    FalloffRegionScaleFactors=(PositiveWidth=1,NegativeWidth=1,PositiveHeight=0.3,NegativeHeight=1)
    FalloffSize=0.15
    OptionalParentItem=(Type=Bone)
    RotationOffset=(X=-110,Y=0,Z=-100)
   )
   > &RigUnit_ModifyTransforms(
    ItemToModify=((Item=(Type=Bone,Name="base_hairB_R"),Transform=(Rotation=(X=-0.2549,Y=0.0449,Z=-0.1677,W=0.9513),Translation=(X=0,Y=0,Z=0),Scale3D=(X=1,Y=1,Z=1))))
    WeightMinimum=0
    Weight=@n2.OutputParam
   )
   &RigUnit_ModifyTransforms(
    ExecutePin=()
    ItemToModify="((Item=(Type=Bone,Name=\"base_hairB_R\"),Transform=(Rotation=(X=0,Y=0,Z=-0.1736,W=0.9848),Translation=(X=0,Y=0,Z=0),Scale3D=(X=1,Y=1,Z=1))),(Item=(Type=Bone,Name=\"dyn_hairB_R_02\"),Transform=(Rotation=(X=0,Y=0,Z=0,W=1),Translation=(X=0,Y=0,Z=0),Scale3D=(X=1,Y ...(truncated)"
    WeightMinimum=0
    Weight=@n2.OutputParam
   )
   RigVMFunction_Sequence_1 = RigVMAggregateNode(ExecuteContext=())
  ]}
 RigVMFunctionLibrary {nodes [
   Controllers = RigVMCollapseNode(Debug=False, ExecuteContext=())
   Correctives = RigVMCollapseNode(ExecuteContext=())
   Set SPR = RigVMCollapseNode(ExecuteContext=())
   exe = RigVMCollapseNode(ExecuteContext=())
  ]}
 RigVMFunction_Sequence_1_1_ContainedGraph {nodes [
   n1 = RigVMFunctionEntryNode()
   n2 = &RigVMFunction_Sequence(ExecuteContext=@n1.ExecuteContext)
   n3 = &RigVMFunction_Sequence(ExecuteContext=@n2.B)
   n4 = &RigVMFunction_Sequence(ExecuteContext=@n3.B)
   RigVMFunctionReturnNode(A=@n2.A, B=@n3.A, C=@n4.A, D=@n4.B)
  ]}
 RigVMFunction_Sequence_1_ContainedGraph {nodes [
   n1 = RigVMFunctionEntryNode()
   RigVMFunction_Sequence = &RigVMFunction_Sequence(ExecuteContext=@n1.ExecuteContext)
   n2 = &RigVMFunction_Sequence(ExecuteContext=@RigVMFunction_Sequence.B)
   RigVMFunctionReturnNode(A=@RigVMFunction_Sequence.A, B=@n2.A, C=@n2.B)
  ]}
 RigVMFunction_Sequence_1_ContainedGraph {nodes [
   n1 = RigVMFunctionEntryNode()
   RigVMFunction_Sequence = &RigVMFunction_Sequence(ExecuteContext=@n1.ExecuteContext)
   n2 = &RigVMFunction_Sequence(ExecuteContext=@RigVMFunction_Sequence.B)
   RigVMFunctionReturnNode(A=@RigVMFunction_Sequence.A, B=@n2.A, C=@n2.B)
  ]}
 RigVMFunction_Sequence_1_ContainedGraph {nodes [
   n1 = RigVMFunctionEntryNode()
   RigVMFunction_Sequence = &RigVMFunction_Sequence(ExecuteContext=@n1.ExecuteContext)
   n2 = &RigVMFunction_Sequence(ExecuteContext=@RigVMFunction_Sequence.B)
   n3 = &RigVMFunction_Sequence(ExecuteContext=@n2.B)
   RigVMFunctionReturnNode(A=@RigVMFunction_Sequence.A, B=@n2.A, C=@n3.A, D=@n3.B)
  ]}
 RigVMModel {nodes [
   RigUnit_BeginExecution = &RigUnit_BeginExecution()
   Controllers = RigVMFunctionReferenceNode(Debug=false, ExecuteContext=@RigUnit_BeginExecution.ExecutePin)
   > Set SPR = RigVMFunctionReferenceNode()
   > exe = RigVMFunctionReferenceNode()
   > Correctives = RigVMFunctionReferenceNode()
  ]}
]
hierarchy [
 root {loc=<0 0 0>, rot=<0 0 0 1>}
 pelvis {parent=root, loc=<0 -0.5301 98.0571>, rot=<0.0127 0.707 -0.0127 -0.707>}
 spine_01 {parent=pelvis, loc=<6.1208 -0.9328 0>, rot=<0 0 0.034 -0.9994>}
 spine_02 {parent=spine_01, loc=<6.5996 -0.2095 0>, rot=<0 0 0.0238 -0.9997>}
 spine_03 {parent=spine_02, loc=<6.4914 -0.1237 0>, rot=<0 0 0.0149 -0.9999>}
 spine_04 {parent=spine_03, loc=<6.9598 0.0794 0>, rot=<0 0 -0.0132 -0.9999>}
 spine_05 {parent=spine_04, loc=<6.9987 0 0>, rot=<0 0 -0.0015 -1>}
 clavicle_l {parent=spine_05, loc=<5.4831 -0.1053 -3.2946>, rot=<0.0465 -0.7289 -0.009 -0.6829>}
 upperarm_l {parent=clavicle_l, loc=<9.8658 -1.2136 -0.4105>, rot=<0.0164 -0.3245 -0.0269 -0.9454>}
 upperarm_twist_01_l {parent=upperarm_l, loc=<7.2135 -0.1134 0>, rot=<0 0 0 1>}
 neck_01 {parent=spine_05, loc=<11.3138 0.1529 0>, rot=<0 0 -0.2003 -0.9797>}
 neck_02 {parent=neck_01, loc=<4.5627 0 0>, rot=<0 0 0.0504 -0.9987>}
 head {parent=neck_02, loc=<5.3914 0 0>, rot=<0 0 0.115 -0.9934>}
 base_hairB_C {parent=head, loc=<7.45 -7.3993 0>, rot=<0.0655 -0.9978 0 0>}
 dyn_hairB_C_01 {parent=base_hairB_C, loc=<0 0 0>, rot=<0 0 0 1>}
 dyn_hairB_C_02 {parent=dyn_hairB_C_01, loc=<7.1108 0 0>, rot=<0 0 0.0914 -0.9958>}
 dyn_hairB_C_03 {parent=dyn_hairB_C_02, loc=<7.1108 0 0>, rot=<0 0 -0.0052 -1>}
 dyn_hairB_C_04 {parent=dyn_hairB_C_03, loc=<7.1108 0 0>, rot=<0 0 -0.1244 -0.9922>}
 dyn_hairB_C_05 {parent=dyn_hairB_C_04, loc=<7.1108 0 0>, rot=<0 0 0 1>}
 dyn_hairB_C_04_sec {parent=dyn_hairB_C_04, loc=<-0.0301 -1.2656 0.4407>, rot=<0 0 0.1192 -0.9929>}
 base_hairB_R {parent=head, loc=<7.3816 -3.637 6.4245>, rot=<0.1412 -0.99 0 0>}
 dyn_hairB_R_01 {parent=base_hairB_R, loc=<0 0 0>, rot=<0 0 0 1>}
 dyn_hairB_R_02 {parent=dyn_hairB_R_01, loc=<7.1108 0 0>, rot=<0 0 0.0675 -0.9977>}
 dyn_hairB_R_03 {parent=dyn_hairB_R_02, loc=<7.1108 0 0>, rot=<0 0 -0.0006 -1>}
 dyn_hairB_R_04 {parent=dyn_hairB_R_03, loc=<7.1108 0 0>, rot=<0 0 -0.1331 -0.9911>}
 dyn_hairB_R_05 {parent=dyn_hairB_R_04, loc=<7.1108 0 0>, rot=<0 0 0 1>}
 dyn_hairB_R_08 {parent=dyn_hairB_R_03, loc=<5.4109 1.8937 -5.2224>, rot=<-0.1178 -0.0936 0.0162 -0.9885>}
 dyn_hairB_R_03_sec {parent=dyn_hairB_R_03, loc=<-0.5867 -0.9471 -0.3317>, rot=<-0.0179 -0.213 0.2817 -0.9354>}
 base_hairF_R_a {parent=head, loc=<4.5798 3.9678 5.9527>, rot=<0.0647 -0.6807 0.726 -0.0735>}
 dyn_hairF_R_a01 {parent=base_hairF_R_a, loc=<0 0 0>, rot=<0 0 0 1>}
 dyn_hairF_R_b01 {parent=dyn_hairF_R_a01, loc=<3.7584 -0.9855 1.4419>, rot=<0.0418 0.2485 0.2153 -0.9435>}
 dyn_hairF_R_b02 {parent=dyn_hairF_R_b01, loc=<8.1922 0 0>, rot=<0.0026 0.0219 0.0147 -0.9996>}
 dyn_hairF_R_a02 {parent=dyn_hairF_R_a01, loc=<10.789 -0.0066 0.0976>, rot=<-0.0065 -0.1657 0.0136 -0.9861>}
 dyn_hairF_R_a03 {parent=dyn_hairF_R_a02, loc=<7.1108 0 0>, rot=<-0.0013 -0.3005 0.0355 -0.9531>}
 dyn_hairF_R_a04 {parent=dyn_hairF_R_a03, loc=<5.0301 0 0>, rot=<0 -0.1603 0 -0.9871>}
 base_hairB_L {parent=head, loc=<7.3811 -3.637 -6.4246>, rot=<0.1412 -0.99 0 0>}
 dyn_hairB_L_01 {parent=base_hairB_L, loc=<0 0 0>, rot=<0 0 0 1>}
 dyn_hairB_L_02 {parent=dyn_hairB_L_01, loc=<7.1102 -0.0002 0>, rot=<0 0 0.0675 -0.9977>}
 dyn_hairB_L_03 {parent=dyn_hairB_L_02, loc=<7.1114 0.0002 0>, rot=<0 0 -0.0006 -1>}
 dyn_hairB_L_04 {parent=dyn_hairB_L_03, loc=<7.11 -0.0003 0>, rot=<0 0 -0.1331 -0.9911>}
 dyn_hairB_L_05 {parent=dyn_hairB_L_04, loc=<7.1113 0 0>, rot=<0 0 0 -1>}
 dyn_hairB_L_03_sec {parent=dyn_hairB_L_04, loc=<-0.4151 -1.6117 -0.3509>, rot=<0 0 0.1607 -0.987>}
 dyn_hairB_L_06 {parent=dyn_hairB_L_02, loc=<7.8621 3.416 3.9255>, rot=<0.2182 0.1409 0.0876 -0.9617>}
 dyn_hairB_L_07 {parent=dyn_hairB_L_06, loc=<6.6366 -0.0002 -0.1721>, rot=<0 0 0 -1>}
 dyn_hairF_L_b01 {parent=dyn_hairB_L_01, loc=<4.669 5.3625 1.222>, rot=<0.6695 -0.0017 0.2436 -0.7017>}
 dyn_hairF_L_b02 {parent=dyn_hairF_L_b01, loc=<7.9248 0 0>, rot=<0.0028 -0.0924 0.1644 -0.9821>}
 base_hairF_L_a {parent=head, loc=<2.3409 3.2804 -6.1487>, rot=<0.6809 0.0614 -0.056 -0.7277>}
 dyn_hairF_L_a01 {parent=base_hairF_L_a, loc=<0 0 0>, rot=<0 0 0 1>}
 dyn_hairF_L_a02 {parent=dyn_hairF_L_a01, loc=<-8.2125 0 0>, rot=<0.1379 -0.0062 0.9903 0.0138>}
 dyn_hairF_L_a03 {parent=dyn_hairF_L_a02, loc=<7.1108 0 0>, rot=<-0.0013 0.2311 0.0355 -0.9723>}
 dyn_hairF_L_a04 {parent=dyn_hairF_L_a03, loc=<5.0301 0 0>, rot=<0 0.1603 0 -0.9871>}
 base_hairbangs_01_L {parent=head, loc=<7.2292 8.9203 -4.4495>, rot=<0.1626 -0.969 0.0292 -0.1836>}
 dyn_hairbangs_01_L {parent=base_hairbangs_01_L, loc=<0 0 0>, rot=<0 0 0 1>}
 dyn_hairbangs_02_L {parent=dyn_hairbangs_01_L, loc=<9.19 -0.2681 -0.3412>, rot=<0.0857 0.2401 -0.1106 -0.9606>}
 base_hairbangs_01_R {parent=head, loc=<6.3123 8.3281 4.3459>, rot=<-0.0194 -0.7068 -0.6512 0.2756>}
 dyn_hairbangs_01_R {parent=base_hairbangs_01_R, loc=<0 0 0>, rot=<0 0 0 1>}
 dyn_hairbangs_02_R {parent=dyn_hairbangs_01_R, loc=<6.0157 0 0>, rot=<-0.0451 -0.0725 -0.1277 -0.9881>}
 base_hairB_R_06 {parent=head, loc=<3.8036 1.8042 6.8978>, rot=<0.3402 -0.6603 0.6626 -0.0962>}
 dyn_hairB_R_06 {parent=base_hairB_R_06, loc=<0 0 0>, rot=<0 0 0 1>}
 dyn_hairB_R_07 {parent=dyn_hairB_R_06, loc=<11.2504 1.3056 -0.7514>, rot=<0.5076 -0.0658 -0.0027 -0.8591>}
 clavicle_r {parent=spine_05, loc=<5.4831 -0.1053 3.2946>, rot=<0.7289 0.0465 -0.6829 0.009>}
 upperarm_r {parent=clavicle_r, loc=<-9.8658 1.2136 0.4105>, rot=<0.0164 -0.3245 -0.0269 -0.9454>}
 upperarm_twist_01_r {parent=upperarm_r, loc=<-7.2135 0.1134 0>, rot=<0 0 0 -1>}
 head_ctrl {controlType=EULER_TRANSFORM, type=Control, loc=<0 -1.6873 152.0342>, rot=<0.0029 -0.7071 -0.0029 0.7071>}
 clavicle_l_ctrl {controlType=EULER_TRANSFORM, type=Control, loc=<3.2946 -4.1138 136.616>, rot=<0.0008 0.0336 -0.0252 0.9991>}
 upperarm_l_ctrl {controlType=EULER_TRANSFORM, parent=clavicle_l_ctrl, type=Control, loc=<13.037 -5.8204 135.5446>, rot=<-0.0065 0.3563 0.0039 0.9343>}
 clavicle_r_ctrl {controlType=EULER_TRANSFORM, type=Control, loc=<75.5157 -3.7455 31.4014>, rot=<0.0034 -0.3248 -0.0273 0.9454>}
 upperarm_r_ctrl {controlType=EULER_TRANSFORM, parent=clavicle_r_ctrl, type=Control, loc=<-13.037 -5.8204 135.5447>, rot=<0.9343 -0.0039 0.3563 0.0065>}
]
variables [
 PickaxeSwing {type=double, default=0}
 armUp_l_alpha {type=double, default=0}
 armUp_r_alpha {type=double, default=0}
 head_lookdown_alpha {type=double, default=0}
 head_lookleft_alpha {type=double, default=0}
 head_lookright_alpha {type=double, default=0}
 head_lookup_alpha {type=double, default=0}
 head_up_relative {type=double, default=0}
 pub aimPitch {type=double, default=0}
 pub isEmoting {type=double, default=0}
 pub isTargeting {type=double, default=0}
 pub isUsingLargeWeapons {type=double, default=0}
 pub isUsingMingun {type=double, default=0}
]
