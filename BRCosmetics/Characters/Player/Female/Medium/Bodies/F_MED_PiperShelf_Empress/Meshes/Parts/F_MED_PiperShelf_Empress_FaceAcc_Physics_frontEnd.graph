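# Front-end physics asset for the Piper Shelf Empress face accessory mesh, setting up physics for her back hair, front locks and bangs.
# Capsules cover the head, neck, upper spine, clavicles and upper arms, plus each hair bone.
# Hair chains branch from base bones constrained to the head, using limited swing and mostly zero twist limits.
# Free-motion constraints cross-link the left and right back chains to the center one.
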PhysicsAsset /BRCosmetics/Characters/Player/Female/Medium/Bodies/F_MED_PiperShelf_Empress/Meshes/Parts/F_MED_PiperShelf_Empress_FaceAcc_Physics_frontEnd
bodies {
 base_hairB_C [{shapeType=Capsule, length=0.505, radius=0.505, shapeName=base_hairB_C_capsule, rotation=<rot 0 -90 0>}]
 base_hairB_L [{shapeType=Capsule, length=0.505, radius=0.505, shapeName=base_hairB_L_capsule, rotation=<rot 0 -90 0>}]
 base_hairB_R [{shapeType=Capsule, length=0.505, radius=0.505, shapeName=base_hairB_R_capsule, rotation=<rot 0 -90 0>}]
 base_hairB_R_06 [{shapeType=Capsule, length=0.505, radius=0.505, shapeName=base_hairB_R_06_capsule, rotation=<rot 0 -90 0>}]
 base_hairF_L_a [{shapeType=Capsule, length=1, radius=1, shapeName=base_hairF_L_a_capsule}]
 base_hairF_R_a [{shapeType=Capsule, length=0.505, radius=0.505, shapeName=base_hairF_R_a_capsule}]
 base_hairbangs_01_L [{shapeType=Capsule, length=0.505, radius=0.505, shapeName=base_hairbangs_01_L_capsule, rotation=<rot 0 -90 0>}]
 base_hairbangs_01_R [{shapeType=Capsule, length=0.505, radius=0.505, shapeName=base_hairbangs_01_R_capsule, rotation=<rot 0 -90 0>}]
 clavicle_l [{shapeType=Capsule, length=6.7099, radius=5.566, shapeName=clavicle_l_capsule, center=<5.339 -1.12 -1.7464>, rotation=<rot -0.6 -83.4096 95>}]
 clavicle_r [{
   shapeType=Capsule
   length=15.7099
   radius=5.566
   shapeName=clavicle_l_capsule
   center=<-5.339 1.1544 1.7464>
   rotation=<rot 0.7124 -98.3524 94.9853>
  }]
 dyn_hairB_C_01 [{shapeType=Capsule, length=4.4, radius=0.6, shapeName=dyn_hairB_C_01_capsule, center=<4.0783 -1.3083 0>, rotation=<rot 90 0 0>}]
 dyn_hairB_C_02 [{shapeType=Capsule, length=2.7, radius=2, shapeName=dyn_hairB_C_02_capsule, center=<3.1895 -1.3083 0>}]
 dyn_hairB_C_03 [{shapeType=Capsule, length=2, radius=2.3, shapeName=dyn_hairB_L_03_capsule, center=<3.3405 -1.306 0>, rotation=<rot 90 180 176.5953>}]
 dyn_hairB_C_04 [{shapeType=Capsule, length=3.3, radius=2.4, shapeName=dyn_hairB_L_04_capsule, center=<3.5413 -1.5563 -0.0172>, rotation=<rot 90 180 176.5952>}]
 dyn_hairB_C_05 [{shapeType=Capsule, length=5.6, radius=2.6, shapeName=dyn_hairB_C_05_capsule, center=<7.6992 -1.3083 0>, rotation=<rot 90 0 0>}]
 dyn_hairB_L_01 [{shapeType=Capsule, length=2.9, radius=0.5, shapeName=dyn_hairB_L_01_capsule, center=<5.339 -1.3083 0>, rotation=<rot 90 0 0>}]
 dyn_hairB_L_02 [{shapeType=Capsule, length=3.6, radius=2.9, shapeName=dyn_hairB_R_02_capsule, center=<3.1895 -2.0398 1.0176>, rotation=<rot 0 0 26.513>}]
 dyn_hairB_L_03 [{shapeType=Capsule, length=1.4, radius=2.9, shapeName=dyn_hairB_L_03_capsule, center=<3.3922 -2.1756 0>, rotation=<rot 90 180 176.5953>}]
 dyn_hairB_L_04 [
  {
   shapeType=Capsule
   length=3.3
   radius=2.4
   shapeName=dyn_hairB_L_04_capsule
   center=<3.5442 -1.6055 -0.0172>
   rotation=<rot 90 179.1048 175.7001>
  }
  {shapeType=Capsule, length=4.7, radius=2.2, shapeName=dyn_hairB_L_04_capsule, center=<4.0421 0.0638 3.5076>, rotation=<rot 90 180 176.5952>}
 ]
 dyn_hairB_L_05 [{shapeType=Capsule, length=2.6, radius=2.9, shapeName=dyn_hairB_L_05_capsule, center=<3.2493 -2.306 0>, rotation=<rot 90 180 176.5953>}]
 dyn_hairB_L_06 [{shapeType=Capsule, length=2.5, radius=2.4, shapeName=dyn_hairB_L_06_capsule, center=<2.184 -2.3265 0.7>, rotation=<rot 90 180 -162.6665>}]
 dyn_hairB_L_07 [{shapeType=Capsule, length=2, radius=0.9, shapeName=dyn_hairB_L_07_capsule, center=<2.5507 0.0055 -0.6872>, rotation=<rot 67.2291 30.983 0>}]
 dyn_hairB_R_01 [{shapeType=Capsule, length=2.9, radius=0.5, shapeName=dyn_hairB_R_01_capsule, center=<5.1192 -0.5068 0>, rotation=<rot 90 0 0>}]
 dyn_hairB_R_02 [{shapeType=Capsule, length=3.9, radius=2.6, shapeName=dyn_hairB_R_02_capsule, center=<3.1895 -0.4535 -0.2262>, rotation=<rot 0 0 -26.513>}]
 dyn_hairB_R_03 [
  {
   shapeType=Capsule
   length=4.5
   radius=2.9
   shapeName=dyn_hairB_R_03_capsule
   center=<2.6059 2.068 -1.2081>
   rotation=<rot -9.3335 4.7116 -18.9833>
  }
  {shapeType=Capsule, length=2, radius=2, shapeName=dyn_hairB_R_03_capsule, center=<6.7939 0.1397 -4.9918>, rotation=<rot -6.9713 0.0001 0>}
  {
   shapeType=Capsule
   length=4.0582
   radius=1.7635
   shapeName=dyn_hairB_R_03_capsule
   center=<11.1019 2.0335 -5.9623>
   rotation=<rot 84.8717 0.0001 -17.5173>
  }
 ]
 dyn_hairB_R_04 [
  {shapeType=Capsule, length=3.3, radius=2.4, shapeName=dyn_hairB_L_04_capsule, center=<3.482 -0.5868 0>, rotation=<rot 90 180 176.5952>}
  {
   shapeType=Capsule
   length=3.3
   radius=2.4
   shapeName=dyn_hairB_L_04_capsule
   center=<3.4943 -0.7944 -2.4572>
   rotation=<rot 73.4479 -176.5952 180>
  }
 ]
 dyn_hairB_R_05 [{shapeType=Capsule, length=6.9, radius=2, shapeName=dyn_hairB_R_05_capsule, center=<3.0687 0.6085 0>, rotation=<rot -55.7146 4.6151 0>}]
 dyn_hairB_R_06 [{
   shapeType=Capsule
   length=5.6
   radius=2.3
   shapeName=dyn_hairF_L_b01_capsule
   center=<6.5342 1.2666 1.2166>
   rotation=<rot 83.1211 8.359 5.6281>
  }]
 dyn_hairB_R_07 [{
   shapeType=Capsule
   length=5.3
   radius=2.1393
   shapeName=dyn_hairB_R_07_capsule
   center=<4.2121 -2.3432 1.1187>
   rotation=<rot 90 180 -174.8445>
  }]
 dyn_hairB_R_08 [{
   shapeType=Capsule
   length=4.4415
   radius=2.5
   shapeName=dyn_hairB_R_08_capsule
   center=<4.4033 1.0632 4.7222>
   rotation=<rot 90 -179.5524 -178.2293>
  }]
 dyn_hairF_L_a01 [{shapeType=Capsule, length=4.2237, radius=0.6877, shapeName=dyn_hairF_R_a01_capsule, center=<-4.0945 0 0>, rotation=<rot 90 0 180>}]
 dyn_hairF_L_a02 [{
   shapeType=Capsule
   length=4.5275
   radius=1.0388
   shapeName=dyn_hairF_R_a02_capsule
   center=<3.281 0.1405 0>
   rotation=<rot 89.884 -0.0008 179.9992>
  }]
 dyn_hairF_L_a03 [{shapeType=Capsule, length=2.5519, radius=1.0108, shapeName=dyn_hairF_R_a03_capsule, center=<2.4618 0 0.0795>, rotation=<rot 90 180 0>}]
 dyn_hairF_L_a04 [{shapeType=Capsule, length=5.0286, radius=0.9618, shapeName=dyn_hairF_R_a04_capsule, center=<3.2613 0 0.1133>, rotation=<rot 90 0 180>}]
 dyn_hairF_L_b01 [{
   shapeType=Capsule
   length=2.1491
   radius=2.477
   shapeName=dyn_hairF_L_b01_capsule
   center=<4.0309 0.0067 -1.3033>
   rotation=<rot 0 93.3354 -84.9458>
  }]
 dyn_hairF_L_b02 [{
   shapeType=Capsule
   length=6.2927
   radius=1.0136
   shapeName=dyn_hairF_L_b02_capsule
   center=<3.3158 0.2891 -2.4468>
   rotation=<rot 1.6316 -91.9801 -61.2654>
  }]
 dyn_hairF_R_a01 [{shapeType=Capsule, length=7.9991, radius=0.9431, shapeName=dyn_hairF_R_a01_capsule, center=<5.2247 -0.1677 0>, rotation=<rot 90 0 180>}]
 dyn_hairF_R_a02 [
  {shapeType=Capsule, length=5.0451, radius=0.9886, shapeName=dyn_hairF_R_a02_capsule, center=<3.1775 0 0>, rotation=<rot 90 0 0>}
  {
   shapeType=Capsule
   length=1.9546
   radius=0.8801
   shapeName=dyn_hairF_R_a02_capsule
   center=<5.4798 -2.1368 0.5723>
   rotation=<rot 3.9915 36.2261 -69.3089>
  }
 ]
 dyn_hairF_R_a03 [{
   shapeType=Capsule
   length=4.9821
   radius=1.1044
   shapeName=dyn_hairF_R_a03_capsule
   center=<2.7881 0.3103 -0.4797>
   rotation=<rot 87.2551 179.9999 179.9999>
  }]
 dyn_hairF_R_a04 [{
   shapeType=Capsule
   length=3.8919
   radius=1.2
   shapeName=dyn_hairF_R_a04_capsule
   center=<3.7026 0.3722 0.1717>
   rotation=<rot 78.1609 180 -180>
  }]
 dyn_hairF_R_b01 [{
   shapeType=Capsule
   length=3.4061
   radius=0.7879
   shapeName=dyn_hairF_R_b01_capsule
   center=<3.6901 0.1367 -0.2779>
   rotation=<rot 0 -89.6459 -89.2303>
  }]
 dyn_hairF_R_b02 [{
   shapeType=Capsule
   length=2.3398
   radius=0.618
   shapeName=dyn_hairF_R_b02_capsule
   center=<1.6014 -0.7651 -0.0829>
   rotation=<rot 4.347 -126.1547 -84.2767>
  }]
 dyn_hairbangs_01_L [{
   shapeType=Capsule
   length=7.2748
   radius=0.508
   shapeName=dyn_hairbangs_01_L_capsule
   center=<4.6134 0.5534 0>
   rotation=<rot 84.8887 -0.2063 -0.2726>
  }]
 dyn_hairbangs_01_R [{
   shapeType=Capsule
   length=5.0489
   radius=0.5
   shapeName=dyn_hairbangs_01_R_capsule
   center=<2.7865 -0.135 0.0943>
   rotation=<rot 0 -90.2806 -88.2131>
  }]
 dyn_hairbangs_02_L [{
   shapeType=Capsule
   length=3.1924
   radius=0.5
   shapeName=dyn_hairbangs_02_L_capsule
   center=<2.8738 -0.0256 0.1998>
   rotation=<rot 0 97.2626 -78.9642>
  }]
 dyn_hairbangs_02_R [{
   shapeType=Capsule
   length=5.665
   radius=0.5529
   shapeName=dyn_hairbangs_02_R_capsule
   center=<3.2502 -0.0818 0.0032>
   rotation=<rot 0.0102 -91.4423 -90.0571>
  }]
 head [{shapeType=Capsule, length=1.9, radius=8.1, center=<2.5678 1.1032 0>, rotation=<rot 90 0 0>}]
 neck_01 [
  {shapeType=Capsule, length=6.0267, radius=4.9765, shapeName=neck_01_capsule, center=<1.2134 0 0>, rotation=<rot 90 0 0>}
  {shapeType=Capsule, length=27, radius=2, shapeName=neck_01_capsule, center=<-4.654 3.4752 0>}
 ]
 neck_02 [{shapeType=Capsule, length=0, radius=15, shapeName=neck_02_capsule, center=<-7.9344 11.4428 0>}]
 spine_02 [{shapeType=Capsule, length=2.3983, radius=9.3, shapeName=spine_02_capsule, center=<6.0567 -1.4271 0>}]
 spine_03 [{shapeType=Capsule, length=4.9363, radius=9.8, shapeName=spine_03_capsule, center=<3.1064 -0.3371 0>, rotation=<rot 0 -90 0>}]
 spine_04 [
  {shapeType=Capsule, length=10.6, radius=8, shapeName=spine_04_capsule, center=<1.3013 5.6256 0>}
  {shapeType=Capsule, length=11.1014, radius=6.1141, shapeName=spine_04_capsule, center=<5.189 3.4956 0>}
 ]
 spine_05 [
  {shapeType=Capsule, length=11.1949, radius=7.6319, shapeName=spine_05_capsule, center=<1.5508 -0.9791 0>, rotation=<rot 0 -90 0>}
  {shapeType=Capsule, length=3.7572, radius=4.2087, shapeName=spine_05_capsuleB, center=<1.6258 3.6342 0.0089>, rotation=<rot 0 -34.5303 -90>}
 ]
 upperarm_l [{shapeType=Capsule, length=15.5, radius=5.5494, shapeName=upperarm_l_capsule, center=<10.1226 -0.5592 -0.8718>, rotation=<rot 90 0 0>}]
 upperarm_r [
  {shapeType=Capsule, length=15.5, radius=4.5846, shapeName=upperarm_l_capsule, center=<-10.1226 -0.5592 0.8718>, rotation=<rot 90 0 0>}
  {
   shapeType=Capsule
   length=7.5
   radius=4.5846
   shapeName=upperarm_l_capsule
   center=<-2.5843 -0.5592 0.5211>
   rotation=<rot -5 -179.9999 -179.9999>
  }
 ]
}
constraints [
 {bone1Name=dyn_hairB_C_01, bone2Name=base_hairB_C, swing1LimitDegrees=10, swing2LimitDegrees=0, twistLimitDegrees=0}
 {bone1Name=dyn_hairB_C_02, bone2Name=dyn_hairB_C_01, swing1LimitDegrees=25, swing2LimitDegrees=5, twistLimitDegrees=0}
 {bone1Name=dyn_hairB_C_03, bone2Name=dyn_hairB_C_02, swing1LimitDegrees=25, swing2LimitDegrees=5, twistLimitDegrees=0}
 {bone1Name=dyn_hairB_C_04, bone2Name=dyn_hairB_C_03, swing1LimitDegrees=25, swing2LimitDegrees=8, twistLimitDegrees=0}
 {bone1Name=dyn_hairB_C_05, bone2Name=dyn_hairB_C_04, swing1LimitDegrees=20, swing2LimitDegrees=8, twistLimitDegrees=0}
 {bone1Name=dyn_hairB_L_01, bone2Name=base_hairB_L, swing1LimitDegrees=15, swing2LimitDegrees=0, twistLimitDegrees=0}
 {bone1Name=dyn_hairB_L_02, bone2Name=dyn_hairB_L_01, swing1LimitDegrees=25, swing2LimitDegrees=3, twistLimitDegrees=0}
 {bone1Name=dyn_hairB_L_03, bone2Name=dyn_hairB_L_02, swing1LimitDegrees=25, swing2LimitDegrees=3, twistLimitDegrees=0}
 {bone1Name=dyn_hairB_L_04, bone2Name=dyn_hairB_L_03, swing1LimitDegrees=25, swing2LimitDegrees=5, twistLimitDegrees=0}
 {bone1Name=dyn_hairB_L_05, bone2Name=dyn_hairB_L_04, swing1LimitDegrees=20, swing2LimitDegrees=8, twistLimitDegrees=0}
 {bone1Name=dyn_hairB_L_06, bone2Name=dyn_hairB_L_02, swing1LimitDegrees=20, swing2LimitDegrees=1, twistLimitDegrees=0}
 {bone1Name=dyn_hairB_L_07, bone2Name=dyn_hairB_L_06, swing1LimitDegrees=20, swing2LimitDegrees=10, twistLimitDegrees=0}
 {bone1Name=dyn_hairB_R_01, bone2Name=base_hairB_R, swing1LimitDegrees=10, swing2LimitDegrees=5, twistLimitDegrees=0}
 {bone1Name=dyn_hairB_R_02, bone2Name=dyn_hairB_R_01, swing1LimitDegrees=45, swing2LimitDegrees=5, twistLimitDegrees=0}
 {bone1Name=dyn_hairB_R_03, bone2Name=dyn_hairB_R_02, swing1LimitDegrees=40, swing2LimitDegrees=5, twistLimitDegrees=0}
 {bone1Name=dyn_hairB_R_04, bone2Name=dyn_hairB_R_03, swing1LimitDegrees=25, swing2LimitDegrees=5, twistLimitDegrees=0}
 {bone1Name=dyn_hairB_R_05, bone2Name=dyn_hairB_R_04, swing1LimitDegrees=20, swing2LimitDegrees=8, twistLimitDegrees=0}
 {
  bone1Name=base_hairF_L_a
  bone2Name=head
  swing1LimitDegrees=45
  swing1Motion=Locked
  swing2LimitDegrees=45
  swing2Motion=Locked
  twistLimitDegrees=45
  twistMotion=Locked
 }
 {
  bone1Name=head
  bone2Name=neck_02
  swing1LimitDegrees=45
  swing1Motion=Locked
  swing2LimitDegrees=45
  swing2Motion=Locked
  twistLimitDegrees=45
  twistMotion=Locked
 }
 {
  bone1Name=spine_05
  bone2Name=spine_04
  swing1LimitDegrees=45
  swing1Motion=Locked
  swing2LimitDegrees=45
  swing2Motion=Locked
  twistLimitDegrees=45
  twistMotion=Locked
 }
 {
  bone1Name=clavicle_r
  bone2Name=spine_05
  swing1LimitDegrees=45
  swing1Motion=Locked
  swing2LimitDegrees=45
  swing2Motion=Locked
  twistLimitDegrees=45
  twistMotion=Locked
 }
 {
  bone1Name=upperarm_r
  bone2Name=clavicle_r
  swing1LimitDegrees=45
  swing1Motion=Locked
  swing2LimitDegrees=45
  swing2Motion=Locked
  twistLimitDegrees=45
  twistMotion=Locked
 }
 {
  bone1Name=neck_01
  bone2Name=spine_05
  swing1LimitDegrees=45
  swing1Motion=Locked
  swing2LimitDegrees=45
  swing2Motion=Locked
  twistLimitDegrees=45
  twistMotion=Locked
 }
 {
  bone1Name=spine_04
  bone2Name=spine_03
  swing1LimitDegrees=45
  swing1Motion=Locked
  swing2LimitDegrees=45
  swing2Motion=Locked
  twistLimitDegrees=45
  twistMotion=Locked
 }
 {
  bone1Name=spine_03
  bone2Name=spine_02
  swing1LimitDegrees=45
  swing1Motion=Locked
  swing2LimitDegrees=45
  swing2Motion=Locked
  twistLimitDegrees=45
  twistMotion=Locked
 }
 {
  bone1Name=clavicle_l
  bone2Name=spine_05
  swing1LimitDegrees=45
  swing1Motion=Locked
  swing2LimitDegrees=45
  swing2Motion=Locked
  twistLimitDegrees=45
  twistMotion=Locked
 }
 {
  bone1Name=upperarm_l
  bone2Name=clavicle_l
  swing1LimitDegrees=45
  swing1Motion=Locked
  swing2LimitDegrees=45
  swing2Motion=Locked
  twistLimitDegrees=45
  twistMotion=Locked
 }
 {bone1Name=dyn_hairB_R_06, bone2Name=base_hairB_R_06, swing1LimitDegrees=2, swing2LimitDegrees=30, twistLimitDegrees=0}
 {bone1Name=dyn_hairB_R_08, bone2Name=base_hairB_R, swing1LimitDegrees=20, swing2LimitDegrees=5, twistLimitDegrees=0}
 {bone1Name=dyn_hairB_R_07, bone2Name=dyn_hairB_R_06, swing1LimitDegrees=20, swing2LimitDegrees=2, twistLimitDegrees=0}
 {
  bone1Name=base_hairF_R_a
  bone2Name=head
  swing1LimitDegrees=10
  swing1Motion=Locked
  swing2LimitDegrees=10
  swing2Motion=Locked
  twistLimitDegrees=5
  twistMotion=Locked
 }
 {bone1Name=dyn_hairF_R_a01, bone2Name=base_hairF_R_a, swing1LimitDegrees=15, swing2LimitDegrees=15, twistLimitDegrees=0}
 {bone1Name=dyn_hairF_R_a02, bone2Name=dyn_hairF_R_a01, swing1LimitDegrees=20, swing2LimitDegrees=24, twistLimitDegrees=0}
 {bone1Name=dyn_hairF_R_a03, bone2Name=dyn_hairF_R_a02, swing1LimitDegrees=15, swing2LimitDegrees=24, twistLimitDegrees=0}
 {bone1Name=dyn_hairF_R_a04, bone2Name=dyn_hairF_R_a03, swing1LimitDegrees=10, swing2LimitDegrees=30, twistLimitDegrees=0}
 {bone1Name=dyn_hairF_R_b01, bone2Name=dyn_hairF_R_a01, swing1LimitDegrees=10, swing2LimitDegrees=15, twistLimitDegrees=0}
 {bone1Name=dyn_hairF_R_b02, bone2Name=dyn_hairF_R_b01, swing1LimitDegrees=10, swing2LimitDegrees=10, twistLimitDegrees=0}
 {bone1Name=dyn_hairF_L_a01, bone2Name=base_hairF_L_a, swing1LimitDegrees=3, swing2LimitDegrees=10, twistLimitDegrees=5}
 {bone1Name=dyn_hairF_L_a02, bone2Name=dyn_hairF_L_a01, swing1LimitDegrees=20, swing2LimitDegrees=20, twistLimitDegrees=0}
 {bone1Name=dyn_hairF_L_a03, bone2Name=dyn_hairF_L_a02, swing1LimitDegrees=10, swing2LimitDegrees=20, twistLimitDegrees=0}
 {bone1Name=dyn_hairF_L_a04, bone2Name=dyn_hairF_L_a03, swing1LimitDegrees=10, swing2LimitDegrees=30, twistLimitDegrees=0}
 {bone1Name=dyn_hairF_L_b01, bone2Name=dyn_hairB_L_01, swing1LimitDegrees=5, swing2LimitDegrees=20, twistLimitDegrees=0}
 {bone1Name=dyn_hairF_L_b02, bone2Name=dyn_hairF_L_b01, swing1LimitDegrees=15, swing2LimitDegrees=15, twistLimitDegrees=0}
 {
  bone1Name=base_hairB_C
  bone2Name=head
  swing1LimitDegrees=45
  swing1Motion=Locked
  swing2LimitDegrees=45
  swing2Motion=Locked
  twistLimitDegrees=45
  twistMotion=Locked
 }
 {
  bone1Name=base_hairB_L
  bone2Name=head
  swing1LimitDegrees=45
  swing1Motion=Locked
  swing2LimitDegrees=45
  swing2Motion=Locked
  twistLimitDegrees=45
  twistMotion=Locked
 }
 {
  bone1Name=base_hairB_R
  bone2Name=head
  swing1LimitDegrees=45
  swing1Motion=Locked
  swing2LimitDegrees=45
  swing2Motion=Locked
  twistLimitDegrees=45
  twistMotion=Locked
 }
 {
  bone1Name=dyn_hairB_L_04
  bone2Name=dyn_hairB_C_04
  swing1LimitDegrees=45
  swing1Motion=Free
  swing2LimitDegrees=45
  swing2Motion=Free
  twistLimitDegrees=45
  twistMotion=Free
 }
 {
  bone1Name=dyn_hairB_R_04
  bone2Name=dyn_hairB_C_04
  swing1LimitDegrees=45
  swing1Motion=Free
  swing2LimitDegrees=45
  swing2Motion=Free
  twistLimitDegrees=45
  twistMotion=Free
 }
 {bone1Name=dyn_hairbangs_02_R, bone2Name=dyn_hairbangs_01_R, swing1LimitDegrees=5, swing2LimitDegrees=20, twistLimitDegrees=0}
 {
  bone1Name=base_hairbangs_01_L
  bone2Name=head
  swing1LimitDegrees=45
  swing1Motion=Free
  swing2LimitDegrees=45
  swing2Motion=Free
  twistLimitDegrees=45
  twistMotion=Free
 }
 {bone1Name=dyn_hairbangs_01_L, bone2Name=base_hairbangs_01_L, swing1LimitDegrees=20, swing2LimitDegrees=10, twistLimitDegrees=0}
 {bone1Name=dyn_hairbangs_02_L, bone2Name=dyn_hairbangs_01_L, swing1LimitDegrees=20, swing2LimitDegrees=10, twistLimitDegrees=0}
 {
  bone1Name=base_hairbangs_01_R
  bone2Name=head
  swing1LimitDegrees=45
  swing1Motion=Free
  swing2LimitDegrees=45
  swing2Motion=Free
  twistLimitDegrees=45
  twistMotion=Free
 }
 {bone1Name=dyn_hairbangs_01_R, bone2Name=base_hairbangs_01_R, swing1LimitDegrees=10, swing2LimitDegrees=20, twistLimitDegrees=0}
 {
  bone1Name=base_hairB_R_06
  bone2Name=head
  swing1LimitDegrees=45
  swing1Motion=Free
  swing2LimitDegrees=45
  swing2Motion=Free
  twistLimitDegrees=45
  twistMotion=Free
 }
 {
  bone1Name=neck_02
  bone2Name=neck_01
  swing1LimitDegrees=45
  swing1Motion=Free
  swing2LimitDegrees=45
  swing2Motion=Free
  twistLimitDegrees=45
  twistMotion=Free
 }
]
