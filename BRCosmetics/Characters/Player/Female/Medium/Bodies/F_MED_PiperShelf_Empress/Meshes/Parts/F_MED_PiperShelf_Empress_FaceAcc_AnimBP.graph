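# Post-process style AnimBP for the Empress Piper Shelf face-accessory part, driving its dynamic hair bones.
# Copied parent and body poses pass through a Control Rig fed by aiming, weapon, pickaxe and emote state, then into RigidBody simulation.
# With backbling, cape, zipline or front end, output switches to separate simulations plus weighted bone constraints and offsets on the hair.
# The EventGraph sets these flags and swaps sim-space settings for dual-wield melee or tactical jumps.
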AnimBlueprint /BRCosmetics/Characters/Player/Female/Medium/Bodies/F_MED_PiperShelf_Empress/Meshes/Parts/F_MED_PiperShelf_Empress_FaceAcc_AnimBP
target_skeleton=&/F_MED_PiperShelf_Empress_FaceAcc_Skeleton
graphs [
 AnimGraph {nodes [
   n1 = RigidBody!(
    ComponentPose=Local To Component(LocalPose=Use cached pose 'CRCache_Input'())
    Alpha
    OverrideWorldGravity:Vector=Get Gravity Override()
    ComponentLinearAccScale:Vector="0.06,0.06,0.1"
    ComponentLinearVelScale:Vector="0.06,0.06,0.1"
    SimSpaceSettings:Sim Space Settings Structure=Get Sim Space Settings Output() -> (Sim Space Settings Output:Sim Space Settings Structure)
   ) -> (Pose) {
    bClampLinearTranslationLimitToRefPose=true
    bOverrideWorldGravity=true
    bTransferBoneVelocities=true
    overridePhysicsAsset=&/F_MED_PiperShelf_Empress_FaceAcc_Physics
    simulationSpace=BaseBoneSpace
    alphaBoolBlend {blendOption=Linear}
    baseBoneRef {boneName=spine_05}
    componentAppliedLinearAccClamp=<3000 3000 3000>
    componentLinearAccScale=<0.04 0.04 0.1>
    componentLinearVelScale=<0.04 0.04 0.1>
    simSpaceSettings {worldAlpha=0.3}
   }
   Output Pose(Result=Blend Poses by bool(
     BlendPose_0=Use cached pose 'CRCache_constraintPose'()
     BlendPose_1=Component To Local(ComponentPose=@n1)
     bActiveValue=Get toggleConstraints() -> (toggleConstraints:bool)
    ) {blendType=Linear}) {layerGroup=None}
   n2 = Layered blend per bone(
    BasePose=Copy Pose From Mesh() {bCopyCurves=true, bUseAttachedParent=true}
    BlendPoses_0=Copy Pose From Mesh(SourceMeshComponent=Get body Mesh() -> (body Mesh:Skeletal Mesh Component Object Reference))
   ) {layerSetup [{branchFilters [{boneName=thigh_l}, {boneName=thigh_r}, {boneName=calf_l}, {boneName=calf_r}]}]}
   n3 = Control Rig!(
    Source=@n2
    Alpha
    pub isTargeting:double=Get isTargeting Float() -> (isTargeting Float:double)
    pub isUsingLargeWeapons:double=Get isUsingLargeWeapons() -> (isUsingLargeWeapons:double)
    pub isUsingMingun:double=Get isUsingMingun() -> (isUsingMingun:double)
    pub isEmoting:double=Get isPlayingEmote Float() -> (isPlayingEmote Float:double)
    pub aimPitch:double=Get AimPitch()
    PickaxeSwing:double=Get PickaxeSwing() -> (PickaxeSwing:double)
   ) -> (Pose) {
    alphaBoolBlend {blendOption=Linear}
    controlRigAssetReference {blueprintRigClass=&/F_MED_PiperShelf_Empress_FaceAcc_CtrlRig.F_MED_PiperShelf_Empress_FaceAcc_CtrlRig_C}
   }
   Save cached pose 'CRCache_Input'(Pose=@n3) {cacheName=CRCache_Input}
   n4 = "Constraint - dyn_hairB_L_03 (PPPPP)"!(
    ComponentPose=Local To Component(LocalPose=Use cached pose 'CRCache_RBSel'())
    Alpha=1
    ConstraintWeights_0=0.5427
    ConstraintWeights_1:float=0.3198
    ConstraintWeights_2:float=0.0611
    ConstraintWeights_3:float=0.0425
    ConstraintWeights_4:float=0.0339
   ) -> (Pose) {
    alphaBoolBlend {blendOption=Linear}
    alphaScaleBiasClamp {bInterpResult=true, interpSpeedDecreasing=5, interpSpeedIncreasing=5}
    boneToModify {boneName=dyn_hairB_L_03}
    constraintSetup [
     {targetBone {boneName=spine_05}}
     {targetBone {boneName=neck_01}}
     {targetBone {boneName=clavicle_l}}
     {targetBone {boneName=neck_02}}
     {targetBone {boneName=head}}
    ]
    constraintWeights [1,1,1,1,1]
   }
   n5 = "Constraint - dyn_hairB_C_03 (PPPPPPP)"!(
    ComponentPose=@n4
    Alpha=1
    ConstraintWeights_0=0.5889
    ConstraintWeights_1:float=0.3292
    ConstraintWeights_2:float=0.0368
    ConstraintWeights_3:float=0.0292
    ConstraintWeights_4:float=0.0079
    ConstraintWeights_5:float=0.0079
    ConstraintWeights_6:float=0.0001
   ) -> (Pose) {
    alphaBoolBlend {blendOption=Linear}
    alphaScaleBiasClamp {bInterpResult=true, interpSpeedDecreasing=5, interpSpeedIncreasing=5}
    boneToModify {boneName=dyn_hairB_C_03}
    constraintSetup [
     {targetBone {boneName=spine_05}}
     {targetBone {boneName=neck_01}}
     {targetBone {boneName=neck_02}}
     {targetBone {boneName=head}}
     {targetBone {boneName=clavicle_l}}
     {targetBone {boneName=clavicle_r}}
     {targetBone {boneName=spine_04}}
    ]
    constraintWeights [1,1,1,1,1,1,1]
   }
   n6 = "Constraint - dyn_hairB_R_03 (PPPPP)"!(
    ComponentPose=@n5
    Alpha=1
    ConstraintWeights_0=0.5422
    ConstraintWeights_1:float=0.3202
    ConstraintWeights_2:float=0.0609
    ConstraintWeights_3:float=0.0427
    ConstraintWeights_4:float=0.034
   ) -> (Pose) {
    alphaBoolBlend {blendOption=Linear}
    alphaScaleBiasClamp {bInterpResult=true, interpSpeedDecreasing=5, interpSpeedIncreasing=5}
    boneToModify {boneName=dyn_hairB_R_03}
    constraintSetup [
     {targetBone {boneName=spine_05}}
     {targetBone {boneName=neck_01}}
     {targetBone {boneName=clavicle_r}}
     {targetBone {boneName=neck_02}}
     {targetBone {boneName=head}}
    ]
    constraintWeights [1,1,1,1,1]
   }
   n7 = "Constraint - dyn_hairB_L_04 (PP)"!(ComponentPose=@n6, Alpha=1, ConstraintWeights_0=0.6777, ConstraintWeights_1:float=0.3223) -> (Pose) {
    alphaBoolBlend {blendOption=Linear}
    alphaScaleBiasClamp {bInterpResult=true, interpSpeedDecreasing=5, interpSpeedIncreasing=5}
    boneToModify {boneName=dyn_hairB_L_04}
    constraintSetup [{targetBone {boneName=clavicle_l}}, {targetBone {boneName=spine_05}}]
    constraintWeights [1,1]
   }
   n8 = "Constraint - dyn_hairB_C_04 (PPP)"!(ComponentPose=@n7, Alpha=1, ConstraintWeights_0=0.8723, ConstraintWeights_1:float=0.0638, ConstraintWeights_2:float=0.0638) -> (Pose) {
    alphaBoolBlend {blendOption=Linear}
    alphaScaleBiasClamp {bInterpResult=true, interpSpeedDecreasing=5, interpSpeedIncreasing=5}
    boneToModify {boneName=dyn_hairB_C_04}
    constraintSetup [{targetBone {boneName=spine_05}}, {targetBone {boneName=clavicle_l}}, {targetBone {boneName=clavicle_r}}]
    constraintWeights [1,1,1]
   }
   n9 = "Constraint - dyn_hairB_R_04 (PP)"!(ComponentPose=@n8, Alpha=1, ConstraintWeights_0=0.6777, ConstraintWeights_1:float=0.3223) -> (Pose) {
    alphaBoolBlend {blendOption=Linear}
    alphaScaleBiasClamp {bInterpResult=true, interpSpeedDecreasing=5, interpSpeedIncreasing=5}
    boneToModify {boneName=dyn_hairB_R_04}
    constraintSetup [{targetBone {boneName=clavicle_r}}, {targetBone {boneName=spine_05}}]
    constraintWeights [1,1]
   }
   n10 = "Constraint - dyn_hairB_L_05 (PPP)"!(ComponentPose=@n9, Alpha=1, ConstraintWeights_0=0.5946, ConstraintWeights_1:float=0.3983, ConstraintWeights_2:float=0.0071) -> (Pose) {
    alphaBoolBlend {blendOption=Linear}
    alphaScaleBiasClamp {bInterpResult=true, interpSpeedDecreasing=5, interpSpeedIncreasing=5}
    boneToModify {boneName=dyn_hairB_L_05}
    constraintSetup [{targetBone {boneName=clavicle_l}}, {targetBone {boneName=spine_05}}, {targetBone {boneName=spine_04}}]
    constraintWeights [1,1,1]
   }
   n11 = "Constraint - dyn_hairB_C_05 (PPPPP)"!(
    ComponentPose=@n10
    Alpha=1
    ConstraintWeights_0=0.7336
    ConstraintWeights_1:float=0.1303
    ConstraintWeights_2:float=0.068
    ConstraintWeights_3:float=0.068
    ConstraintWeights_4:float=0.0001
   ) -> (Pose) {
    alphaBoolBlend {blendOption=Linear}
    alphaScaleBiasClamp {bInterpResult=true, interpSpeedDecreasing=5, interpSpeedIncreasing=5}
    boneToModify {boneName=dyn_hairB_C_05}
    constraintSetup [
     {targetBone {boneName=spine_05}}
     {targetBone {boneName=spine_04}}
     {targetBone {boneName=clavicle_l}}
     {targetBone {boneName=clavicle_r}}
     {targetBone {boneName=spine_03}}
    ]
    constraintWeights [1,1,1,1,1]
   }
   n12 = "Constraint - dyn_hairB_R_05 (PPP)"!(ComponentPose=@n11, Alpha=1, ConstraintWeights_0=0.5946, ConstraintWeights_1:float=0.3983, ConstraintWeights_2:float=0.0071) -> (Pose) {
    alphaBoolBlend {blendOption=Linear}
    alphaScaleBiasClamp {bInterpResult=true, interpSpeedDecreasing=5, interpSpeedIncreasing=5}
    boneToModify {boneName=dyn_hairB_R_05}
    constraintSetup [{targetBone {boneName=clavicle_r}}, {targetBone {boneName=spine_05}}, {targetBone {boneName=spine_04}}]
    constraintWeights [1,1,1]
   }
   n13 = "Transform (Modify) Bone - Bone: dyn_hairB_L_03_sec"!(ComponentPose=@n12, bAlphaBoolEnabled:bool=True, Translation="0,0,0", Rotation="0,80,0", Scale="1,1,1") -> (Pose) {
    alphaInputType=Bool
    rotationMode=BMM_Additive
    rotationSpace=BCS_BoneSpace
    scaleMode=BMM_Additive
    scaleSpace=BCS_BoneSpace
    translationMode=BMM_Additive
    translationSpace=BCS_BoneSpace
    alphaBoolBlend {blendOption=Linear}
    boneToModify {boneName=dyn_hairB_L_03_sec}
    rotation {yaw=80}
   }
   n14 = "Transform (Modify) Bone - Bone: dyn_hairB_C_04_sec"!(ComponentPose=@n13, bAlphaBoolEnabled:bool=True, Translation="0,0,0", Rotation="0,70,0", Scale="1,1,1") -> (Pose) {
    alphaInputType=Bool
    rotationMode=BMM_Additive
    rotationSpace=BCS_BoneSpace
    scaleMode=BMM_Additive
    scaleSpace=BCS_BoneSpace
    translationMode=BMM_Additive
    translationSpace=BCS_BoneSpace
    alphaBoolBlend {blendOption=Linear}
    boneToModify {boneName=dyn_hairB_C_04_sec}
    rotation {yaw=70}
   }
   n15 = "Transform (Modify) Bone - Bone: dyn_hairB_R_03_sec"!(ComponentPose=@n14, bAlphaBoolEnabled:bool=True, Translation="0,0,0", Rotation="0,40,0", Scale="1,1,1") -> (Pose) {
    alphaInputType=Bool
    rotationMode=BMM_Additive
    rotationSpace=BCS_BoneSpace
    scaleMode=BMM_Additive
    scaleSpace=BCS_BoneSpace
    translationMode=BMM_Additive
    translationSpace=BCS_BoneSpace
    alphaBoolBlend {blendOption=Linear}
    boneToModify {boneName=dyn_hairB_R_03_sec}
    rotation {yaw=40}
   }
   n16 = "Constraint - dyn_hairB_L_07 (PP)"!(ComponentPose=@n15, Alpha=1, ConstraintWeights_0=0.1, ConstraintWeights_1:float=0.9) -> (Pose) {
    alphaBoolBlend {blendOption=Linear}
    alphaScaleBiasClamp {bInterpResult=true, interpSpeedDecreasing=5, interpSpeedIncreasing=5}
    boneToModify {boneName=dyn_hairB_L_07}
    constraintSetup [{targetBone {boneName=clavicle_l}}, {targetBone {boneName=spine_05}}]
    constraintWeights [1,1]
   }
   n17 = "Constraint - dyn_hairF_L_b01 (PPPP)"!(
    ComponentPose=@n16
    Alpha=1
    ConstraintWeights_0=0.4458
    ConstraintWeights_1:float=0.3967
    ConstraintWeights_2:float=0.1539
    ConstraintWeights_3:float=0.0036
   ) -> (Pose) {
    alphaBoolBlend {blendOption=Linear}
    boneToModify {boneName=dyn_hairF_L_b01}
    constraintSetup [{targetBone {boneName=head}}, {targetBone {boneName=neck_02}}, {targetBone {boneName=neck_01}}, {targetBone {boneName=spine_05}}]
    constraintWeights [1,1,1,1]
   }
   n18 = "Transform (Modify) Bone - Bone: dyn_hairF_L_b02"(ComponentPose=@n17, Alpha=1, Translation="0,0,-3", Rotation="1,0,0", Scale="1,1,1") {
    rotationMode=BMM_Additive
    rotationSpace=BCS_BoneSpace
    scaleMode=BMM_Additive
    scaleSpace=BCS_BoneSpace
    translationMode=BMM_Additive
    translationSpace=BCS_BoneSpace
    alphaBoolBlend {blendOption=Linear}
    boneToModify {boneName=dyn_hairF_L_b02}
    rotation {pitch=1}
    translation {z=-3}
   }
   n19 = "Transform (Modify) Bone - Bone: dyn_hairB_R_03"(ComponentPose=@n18, Alpha=1, Translation="0,-4,0", Rotation="0,0,0", Scale="1,1,1") {
    rotationMode=BMM_Additive
    rotationSpace=BCS_BoneSpace
    scaleMode=BMM_Additive
    scaleSpace=BCS_BoneSpace
    translationMode=BMM_Additive
    translationSpace=BCS_BoneSpace
    alphaBoolBlend {blendOption=Linear}
    boneToModify {boneName=dyn_hairB_R_03}
    translation {y=-4}
   }
   n20 = "Transform (Modify) Bone - Bone: dyn_hairB_R_08"(ComponentPose=@n19, Alpha=1, Translation="0,0,0", Rotation="0,40,0", Scale="1,1,1") {
    rotationMode=BMM_Additive
    rotationSpace=BCS_BoneSpace
    scaleMode=BMM_Additive
    scaleSpace=BCS_BoneSpace
    translationMode=BMM_Additive
    translationSpace=BCS_BoneSpace
    alphaBoolBlend {blendOption=Linear}
    boneToModify {boneName=dyn_hairB_R_08}
    rotation {yaw=40}
   }
   n21 = "Transform (Modify) Bone - Bone: dyn_hairB_R_03"(ComponentPose=@n20, Alpha=1, Translation="0,-5,2", Rotation="0,0,0", Scale="0.9,0.9,0.9") {
    rotationMode=BMM_Additive
    rotationSpace=BCS_BoneSpace
    scaleMode=BMM_Additive
    scaleSpace=BCS_BoneSpace
    translationMode=BMM_Additive
    translationSpace=BCS_BoneSpace
    alphaBoolBlend {blendOption=Linear}
    alphaScaleBiasClamp {bInterpResult=true}
    boneToModify {boneName=dyn_hairB_R_03}
    scale=<0.9 0.9 0.9>
    translation {y=-5, z=2}
   }
   Save cached pose 'CRCache_constraintPose'(Pose=Component To Local(ComponentPose=@n21)) {cacheName=CRCache_constraintPose}
   n22 = RigidBody!(
    ComponentPose=Local To Component(LocalPose=Use cached pose 'CRCache_Input'())
    Alpha
    OverrideWorldGravity="0,0,-980"
    ComponentLinearAccScale="0.01,0.01,0.01"
    ComponentLinearVelScale="0.01,0.01,0.01"
   ) -> (Pose) {
    bClampLinearTranslationLimitToRefPose=true
    bOverrideWorldGravity=true
    bTransferBoneVelocities=true
    overridePhysicsAsset=&/F_MED_PiperShelf_Empress_FaceAcc_Physics_frontEnd
    simulationSpace=BaseBoneSpace
    alphaBoolBlend {blendOption=Linear}
    baseBoneRef {boneName=neck_02}
    componentAppliedLinearAccClamp=<20000 20000 20000>
    componentLinearAccScale=<0.6 0.2 0.2>
    componentLinearVelScale=<0.6 0.2 0.2>
   }
   n23 = RigidBody(
    ComponentPose=Local To Component(LocalPose=Use cached pose 'CRCache_Input'())
    OverrideWorldGravity="0,0,-980"
    ComponentLinearAccScale="0.01,0.01,0.01"
    ComponentLinearVelScale="0.01,0.01,0.01"
   ) {
    bClampLinearTranslationLimitToRefPose=true
    bOverrideWorldGravity=true
    bTransferBoneVelocities=true
    overridePhysicsAsset=&F_MED_PiperShelf_Empress_FaceAcc_Physics
    simulationSpace=BaseBoneSpace
    alphaBoolBlend {blendOption=Linear}
    baseBoneRef {boneName=neck_02}
    componentAppliedLinearAccClamp=<20000 20000 20000>
    componentLinearAccScale=<0.6 0.2 0.2>
    componentLinearVelScale=<0.6 0.2 0.2>
   }
   Save cached pose 'CRCache_RBSel'(Pose=Blend Poses by bool(BlendPose_0=Component To Local(ComponentPose=@n22), BlendPose_1=Component To Local(ComponentPose=@n23), bActiveValue=Get bIsInFrontEnd()) {blendType=Linear}) {cacheName=CRCache_RBSel}
  ]}
 EventGraph {nodes [
   Event BlueprintUpdateAnimation()
   > n1 = Sequence()
   > n2 = Gravity Override State Function DEPRECATED(CharacterPartAnimInstance=Self-Reference(), GameTimeInSeconds=GetGameTimeInSeconds(), GravityOverrideParams=Get Gravity Override Params())
   > Set Gravity Override(Gravity Override=@n2)
   n3 = Get bIsJumping()
   n4 = Get bIsFalling()
   > Set IsTacticalJumping(IsTacticalJumping:bool=AND Boolean(A=Get bIsTacticalSprinting(), B=OR Boolean(A=@n3, B=@n4))) -> (Output_Get:bool)
   > Set toggleConstraints(toggleConstraints:bool=OR Boolean!(
     A=Get isWearingBackbling()
     B=Get onZipline() -> (onZipline:bool)
     C:bool=Get IsWearingCape() -> (IsWearingCape:bool)
     D:bool=Get bIsInFrontEnd()
    ) -> (ReturnValue)) -> (Output_Get:bool)
   > Set PickaxeSwing(PickaxeSwing:double="To Float (Boolean)"(InBool=Get bIsPlayingMeleeAnim())) -> (Output_Get:double)
   n5 = Sequence(execute=@n1.then_1)
   > n6 = Branch(Condition=Get IsWearingCape())
   > Set activateConstraints(activateConstraints:double=1) -> (Output_Get:double)
   > n7 = Branch(Condition=Get bIsSkyDiving())
   > Set isSkydiveCape Float(isSkydiveCape Float:double=1) -> (Output_Get:double)
   Set isSkydiveCape Float(execute=@n7.else, isSkydiveCape Float=0)
   n8 = Branch(execute=@n6.else, Condition=Get bIsSkyDiving())
   > Set activateConstraints(activateConstraints=1)
   > Set isSkydiveBackpack Float(isSkydiveBackpack Float:double=1) -> (Output_Get:double)
   Set activateConstraints(execute=@n8.else, activateConstraints=0)
   > Set isSkydiveBackpack Float(isSkydiveBackpack Float=0)
   n9 = Branch(execute=@n5.then_1, Condition=AND Boolean(A=Get bIsPlayingMeleeAnim(), B=Get isUsingDualWieldMeele() -> (isUsingDualWieldMeele:bool)))
   > Set Sim Space Settings Output(Sim Space Settings Output:Sim Space Settings Structure=Get Sim Space Settings DWMA() -> (Sim Space Settings DWMA:Sim Space Settings Structure)) -> (Output_Get:Sim Space Settings Structure)
   n10 = Branch(execute=@n9.else, Condition=Get IsTacticalJumping() -> (IsTacticalJumping:bool))
   > Set Sim Space Settings Output(Sim Space Settings Output=Get Sim Space Settings Jump() -> (Sim Space Settings Jump:Sim Space Settings Structure))
   Set Sim Space Settings Output(execute=@n10.else, Sim Space Settings Output=Get Sim Space Settings Default() -> (Sim Space Settings Default:Sim Space Settings Structure))
   Event OnBeginPlayingEmote() {is=title K2Node_Event}
   > Set isPlayingEmote Float(isPlayingEmote Float:double=1) -> (Output_Get:double)
   Event OnBeginTargeting() {is=title K2Node_Event}
   > Set isTargeting Float(isTargeting Float:double=1) -> (Output_Get:double)
   Event OnEndPlayingEmote() {is=title K2Node_Event}
   > Set isPlayingEmote Float(isPlayingEmote Float=0)
   Event BlueprintInitializeAnimation()
   > Set body Mesh(body Mesh:Skeletal Mesh Component Object Reference=GetOwnerSkeletalMeshComponent()) -> (Output_Get:Skeletal Mesh Component Object Reference)
   Event OnEndTargeting() {is=title K2Node_Event}
   > Set isTargeting Float(isTargeting Float=0)
   Event OnWeaponChanged()
   > Set isUsingLargeWeapons(isUsingLargeWeapons:double=0) -> (Output_Get:double)
   > Set isUsingMingun(isUsingMingun:double=0) -> (Output_Get:double)
   > Set isUsingDualWieldMeele(isUsingDualWieldMeele:bool=false) -> (Output_Get:bool)
   > n11 = Switch on EFortWeaponCoreAnimation!(execute, Selection:EFortWeaponCoreAnimation Enum=Get WeaponCoreAnim!() -> (WeaponCoreAnim)) -> (
    Melee:Exec
    Pistol:Exec
    Shotgun:Exec
    PaperBlueprint:Exec
    Rifle:Exec
    MeleeOneHand:Exec
    MachinePistol:Exec
    RocketLauncher:Exec
    GrenadeLauncher:Exec
    GoingCommando:Exec
    AssaultRifle:Exec
    TacticalShotgun:Exec
    SniperRifle:Exec
    TrapPlacement:Exec
    ShoulderLauncher:Exec
    AbilityDecoTool:Exec
    Crossbow:Exec
    C4:Exec
    RemoteControl:Exec
    DualWield:Exec
    AR_BullPup:Exec
    AR_ForwardGrip:Exec
    MedPackPaddles:Exec
    SMG_P90:Exec
    AR_DrumGun:Exec
    Consumable_Small:Exec
    Consumable_Large:Exec
    Balloon:Exec
    MountedTurret:Exec
    CreativeTool:Exec
    ExplosiveBow:Exec
    AshtonIndigo:Exec
    AshtonChicago:Exec
    MeleeDualWield:Exec
    Unarmed:Exec
   )
   Set isUsingLargeWeapons(execute=[@n11.AssaultRifle, @n11.Rifle, @n11.RocketLauncher, @n11.Shotgun, @n11.SniperRifle, @n11.TacticalShotgun], isUsingLargeWeapons=1)
   Set isUsingMingun(execute=@n11.GoingCommando, isUsingMingun=1)
   Set isUsingDualWieldMeele(execute=@n11.MeleeDualWield, isUsingDualWieldMeele=true)
   Event OnParachuteOpened()
   > Switch on EGliderType(Selection=Get PlayerGliderType() -> (PlayerGliderType=HangGlider))
   > Set isUsingHangGlider(isUsingHangGlider:bool=true) -> (Output_Get:bool)
   Event OnParachuteClosed()
   > Set isUsingHangGlider(isUsingHangGlider=false)
   Event OnCharacterCustomizationLoaded()
   n12 = CheckForBackbling()
   > Set isWearingBackbling(isWearingBackbling=@n12)
   > Set IsWearingCape(IsWearingCape:bool=OR Boolean!(A=@n12.bIsBackblingACape, B=@n12.bIsBackblingLong) -> (ReturnValue)) -> (Output_Get:bool)
   n13 = Event BlueprintProcessAnimInputEvent()
   > n14 = Branch(Condition=IsNamed(self=@n13.AnimInputEvent, InEventName=aie.Player.Zipline.Entered))
   > Set onZipline(onZipline:bool=1) -> (Output_Get:bool)
   Branch(
    execute=@n14.else
    Condition=OR Boolean!(A=IsNamed(self=@n13.AnimInputEvent, InEventName=aie.Player.Zipline.Exited), B=Get bIsMoving2D(), C:bool=Get bIsJumping()) -> (ReturnValue)
   )
   > Set onZipline(onZipline=false)
  ]}
]
variables [
 activateConstraints {type=number}
 body Mesh {type=SkeletalMeshComponent}
 Delta Time X {type=number}
 Gravity Override {type=Vector}
 Gravity Override Params {type=GravityOverrideParamsStruct}
 isPlayingEmote Float {type=number}
 isSkydiveBackpack Float {type=number}
 isSkydiveCape Float {type=number}
 IsTacticalJumping {type=boolean}
 isTargeting Float {type=number}
 isUsingDualWieldMeele {type=boolean}
 isUsingHangGlider {type=boolean}
 isUsingLargeWeapons {type=number}
 isUsingMingun {type=number}
 isWearingBackbling {type=boolean}
 IsWearingCape {type=boolean}
 lerpInit x2 Float {type=number, default=1}
 onZipline {type=boolean}
 PickaxeSwing {type=number}
 Sim Space Settings Default {
  type=SimSpaceSettings
  default {
   dampingAlpha=1
   maxAngularAcceleration=10000
   maxAngularVelocity=10000
   maxLinearAcceleration=10000
   maxLinearVelocity=10000
   velocityScaleZ=0.5
   worldAlpha=0.4
   externalLinearDragV=<0 0 0>
  }
 }
 Sim Space Settings DWMA {
  type=SimSpaceSettings
  default {
   dampingAlpha=1
   maxAngularAcceleration=10000
   maxAngularVelocity=10000
   maxLinearAcceleration=10000
   maxLinearVelocity=10000
   velocityScaleZ=0
   worldAlpha=0.8
   externalLinearDragV=<0 0 0>
  }
 }
 Sim Space Settings Jump {
  type=SimSpaceSettings
  default {
   dampingAlpha=1
   maxAngularAcceleration=10000
   maxAngularVelocity=10000
   maxLinearAcceleration=10000
   maxLinearVelocity=10000
   velocityScaleZ=0.25
   worldAlpha=0.2
   externalLinearDragV=<0 0 0>
  }
 }
 Sim Space Settings Output {
  type=SimSpaceSettings
  default {
   dampingAlpha=1
   maxAngularAcceleration=10000
   maxAngularVelocity=10000
   maxLinearAcceleration=10000
   maxLinearVelocity=10000
   velocityScaleZ=0
   worldAlpha=1
   externalLinearDragV=<0 0 0>
  }
 }
 toggleConstraints {type=boolean}
]
